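# Lit skeletal-mesh material for the Mastermind Shadow pickaxe, layering goo, vertex displacement and optional gold onto the shared weapon master.
# Weapon attributes pass through wrap customization, with bounds parameters tied in by a zero-alpha lerp, then MF_TwistedGoo masked by vertex colors.
# A panning world-aligned texture, relative to actor position, pushes vertices along normals where vertex red is low.
# A default-black mask blends in gold with texture-driven roughness before tech-art overrides.
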Material /BRCosmetics/Weapons/FORT_Melee/Pickaxe_Mastermind_Shadow/Materials/M_Pickaxe_Mastermind_Shadow_corrected
graph {
 properties {
  bUseMaterialAttributes=true
  bUsedWithMeshDeformer=true
  bUsedWithSkeletalMesh=true
  pixelDepthOffsetMode=PDOM_Legacy
  refractionMethod=RM_IndexOfRefraction
  usedShadingModels=Default Lit
 }
 nodes [
  n1 = MaterialFunctionCall!(In) {materialFunction=/Game/Packages/Fortress_SharedMaterials/Base_Material_Functions/BaseCustomizationFunctions/MF_FN_Weapon_MASTER}
  n2 = GetMaterialAttributes(None=@n1.Material Result) {attributeGetTypes [#1]}
  n3 = VectorParameter() {group=Weapon Specific Wrap Adjustments, parameterName=DO NOT OVERRIDE-OBMAX, defaultValue=<rgba 0 0 0 1>}
  n4 = VectorParameter() {group=Weapon Specific Wrap Adjustments, parameterName=DO NOT OVERRIDE-OBMIN, defaultValue=<rgba 0 0 0 1>}
  n5 = Add(A=@n3.RGB, B=@n4.RGB)
  n6 = VectorParameter() {group=Weapon Specific Wrap Adjustments, parameterName=DO NOT OVERRIDE-OBCENTER, defaultValue=<rgba 0 0 0 1>}
  n7 = SetMaterialAttributes!(MaterialAttributes=@n2.MaterialAttributes, Base Color=LinearInterpolate(A=@n2.Base Color, B=Add(A=@n5, B=@n6.RGB)) {constAlpha=0}) {attributeSetTypes [#1]}
  n8 = VectorParameter() {parameterName=WeaponForwardAxis, defaultValue=<rgba 0 1 0 1>}
  n9 = VertexColor()
  n10 = OneMinus(None=@n9.R)
  n11 = GetMaterialAttributes(None=MaterialFunctionCall!(
    In=MaterialFunctionCall!(
     Material=@n7
     ------------------
     Object Bounds Center Adjustment=@n6.RGB
     Object Bounds Min Adjustment=@n4.RGB
     Wrap ScaleMultiplier
     Rotation X Axis
     "Customization Mask(_CM A)"
     Forward Axis=@n8.RGB
     Object Bounds Max Adjustment=@n3.RGB
     Rotation Z Axis
     Normal Map=@n1.Normal Map
     Cropped Emissive Mask=@n1.Normal Map
     "Customization Crevasse (_CM G)"=@n1.Normal Map
     "Customization Mask UV Islands (_CM B Bottom Right)"=@n1.Normal Map
     -------------------
     "Customization Mask 1 (_CM B Top Left)"=@n1.Normal Map
     "Customization Mask 2 (_CM B Top Right)"=@n1.Normal Map
     "Customization Mask 3 (_CM B Bottom Left)"=@n1.Normal Map
     "Customization Mask (_CM RGBA)"
     --------------------
     Pre-Skinned Local Position=@n1.Normal Map
     Pre-Skinned Bounds=@n1.Normal Map
     Pre-Skinned Local Normal=@n1.Normal Map
     Per_ObjectMaskOverride
     isVehicle
     Side Project Vehicle
     isMech
     isCharacter
     isCharAccent
     isPickaxe
     isBackpack
     Object-SafeZone-Back
     Object-SafeZone-Middle
     Object-SafeZone-Front
     Object-SafeZone-VP_Charcter_Preview_Switch
     Object-SafeZone-Virtual-Pivot
     Wrap Holdout
    ) {materialFunction=/Game/Packages/Fortress_SharedMaterials/Base_Material_Functions/BaseCustomizationFunctions/MF_Base_Customization}.Result
    Mask=Add(A=@n10, B=@n9.G)
    isVehicle
    isMech
   ) {materialFunction=/Game/Characters/Player/Male/Medium/Bodies/M_MED_Mastermind/Materials/MF_TwistedGoo}.Result) {attributeGetTypes [#2]}
  n12 = OneMinus(None=VertexColor().R)
  n13 = ScalarParameter() {defaultValue=-0.1, parameterName=WPO_Offset A}
  n14 = ScalarParameter() {defaultValue=1.5, parameterName=WPO_Offset B}
  n15 = VectorParameter() {parameterName=Panning Speed, defaultValue=<rgba 0 0.1 0 0>}
  n16 = MaterialFunctionCall!(
   TextureObject=TextureObject() {texture=&2ndInfectedMapLinear}
   TextureSize=VectorParameter() {parameterName=Texture Size, defaultValue=<rgba 100 100 100 0>}.RGB
   WorldPosition=Subtract(A=ActorPositionWS(), B=WorldPosition().XYZ)
   Export Float 4
   In=Multiply(A=ComponentMask(None=@n15.RGB) {g=true, r=true}, B=ScalarParameter() {defaultValue=1, parameterName=Panner Speed Scale})
   World Space Normal
   ProjectionTransitionContrast
  ) {materialFunction=/Game/Athena/SafeZone/WorldAlignedTexture_ScalePan}
  n17 = Desaturation(None=@n16.XYZ Texture)
  n18 = Multiply(A=Multiply(A=VertexNormalWS(), B=LinearInterpolate(A=@n13, B=@n14, Alpha=@n17)), B=@n16.XYZ Texture)
  n19 = SetMaterialAttributes!(MaterialAttributes=@n11.MaterialAttributes, World Position Offset=Add(A=@n11.World Position Offset, B=Multiply(A=@n12, B=@n18))) {attributeSetTypes [#2]}
  n20 = VectorParameter() {parameterName=GoldColor, defaultValue=<rgba 0.7656 0.5265 0.1769 1>}
  n21 = Constant() {r=1}
  n22 = LinearInterpolate(
   A=ScalarParameter() {defaultValue=0.1, parameterName=Gold Rougness Min}
   B=ScalarParameter() {defaultValue=1, parameterName=Gold Rougness Max}
   Alpha=ComponentMask(None=@n16.XYZ Texture) {g=true}
  )
  MaterialFunctionCall(In=BlendMaterialAttributes(
    A=@n19
    B=SetMaterialAttributes!(MaterialAttributes=@n19, Base Color=@n20.RGB, Metallic=@n21, Roughness=@n22) {attributeSetTypes [#1, #3, #4]}
    Alpha=TextureSampleParameter2D() {parameterName=Gold_Mask, texture=/Game/Weapons/WeaponSkins/Wraps/Textures/T_SimpleBlack}.R
   )) {materialFunction=&MF_TechArtOverrides_accessories}
 ]
}
